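# Sound Cue for Nitro Queen "Attract" voice lines in NitroDrome.
# A single Random node, excluded from branch culling, picks among five equally weighted wave variants.
# Playback uses the shared NPC alerted attenuation asset and the critical enemy VO sound class.
# It uses the PA announcer VO concurrency set, so its voice limiting follows that asset's rules.
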SoundCue /NitroDrome/Assets/Audio/VO/Nitro_Queen/VO_Nitro_Queen_Attract_Cue
root=Random_0
nodes [
 {
  id=Random_0
  children [WavePlayer_0,WavePlayer_1,WavePlayer_2,WavePlayer_3,WavePlayer_4]
  properties {bShouldExcludeFromBranchCulling=true, weights [1,1,1,1,1]}
 }
 {id=WavePlayer_0, properties {soundWaveAssetPtr=&/VO_Nitro_Queen_Attract_01}}
 {id=WavePlayer_1, properties {soundWaveAssetPtr=&/VO_Nitro_Queen_Attract_02}}
 {id=WavePlayer_2, properties {soundWaveAssetPtr=&/VO_Nitro_Queen_Attract_03}}
 {id=WavePlayer_3, properties {soundWaveAssetPtr=&/VO_Nitro_Queen_Attract_05}}
 {id=WavePlayer_4, properties {soundWaveAssetPtr=&/VO_Nitro_Queen_Attract_08}}
]
properties {
 attenuationSettings=/Game/Sounds/Fort_Human_VO_BR/NPCs/Shared/NPC_VO_Alerted_Att
 maxDistance=12500
 soundClassObject=&Fort_VO_Enemy_Critical
 volumeMultiplier=1
 concurrencySet [/NitroDrome/Assets/Audio/PA/VO/VO_PA_Announcer_Conc]
}
