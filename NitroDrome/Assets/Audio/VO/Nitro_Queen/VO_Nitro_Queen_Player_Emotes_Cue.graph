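# Sound cue for Nitro Queen player emote voice lines in NitroDrome, playing one of five recorded variations.
# A single Random node picks among the five wave players with equal weights and is excluded from branch culling.
# Playback uses the shared NPC alerted attenuation, the critical enemy VO sound class, and the PA announcer VO concurrency set.
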SoundCue /NitroDrome/Assets/Audio/VO/Nitro_Queen/VO_Nitro_Queen_Player_Emotes_Cue
root=Random_5
nodes [
 {
  id=Random_5
  children [WavePlayer_20, WavePlayer_21, WavePlayer_28, WavePlayer_29, WavePlayer_30]
  properties {bShouldExcludeFromBranchCulling=true, weights [1,1,1,1,1]}
 }
 {id=WavePlayer_20, properties {soundWaveAssetPtr=&/VO_Nitro_Queen_Player_Emotes_01}}
 {id=WavePlayer_21, properties {soundWaveAssetPtr=&/VO_Nitro_Queen_Player_Emotes_02}}
 {id=WavePlayer_28, properties {soundWaveAssetPtr=&/VO_Nitro_Queen_Player_Emotes_03}}
 {id=WavePlayer_29, properties {soundWaveAssetPtr=&/VO_Nitro_Queen_Player_Emotes_04}}
 {id=WavePlayer_30, properties {soundWaveAssetPtr=&/VO_Nitro_Queen_Player_Emotes_05}}
]
properties {
 attenuationSettings=/Game/Sounds/Fort_Human_VO_BR/NPCs/Shared/NPC_VO_Alerted_Att
 maxDistance=12500
 soundClassObject=&Fort_VO_Enemy_Critical
 volumeMultiplier=1
 concurrencySet [/NitroDrome/Assets/Audio/PA/VO/VO_PA_Announcer_Conc]
}
