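# Sound cue for Nitro Queen voice lines, named for the moment a player enters a point of interest in a vehicle.
# A Random node picks one of five equally weighted wave players, each holding a variant line, and is excluded from branch culling.
# Playback uses shared NPC alerted attenuation, the critical enemy VO sound class and the PA announcer concurrency set.
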SoundCue /NitroDrome/Assets/Audio/VO/Nitro_Queen/VO_Nitro_Queen_Player_Enters_POI_Veh_Cue
root=Random_1
nodes [
 {
  id=Random_1
  children [WavePlayer_4, WavePlayer_5, WavePlayer_6, WavePlayer_8, WavePlayer_10]
  properties {bShouldExcludeFromBranchCulling=true, weights [1,1,1,1,1]}
 }
 {id=WavePlayer_10, properties {soundWaveAssetPtr=&/VO_Nitro_Queen_Player_Enters_Veh_05}}
 {id=WavePlayer_4, properties {soundWaveAssetPtr=&/VO_Nitro_Queen_Player_Enters_Veh_01}}
 {id=WavePlayer_5, properties {soundWaveAssetPtr=&/VO_Nitro_Queen_Player_Enters_Veh_02}}
 {id=WavePlayer_6, properties {soundWaveAssetPtr=&/VO_Nitro_Queen_Player_Enters_Veh_03}}
 {id=WavePlayer_8, properties {soundWaveAssetPtr=&/VO_Nitro_Queen_Player_Enters_Veh_04}}
]
properties {
 attenuationSettings=/Game/Sounds/Fort_Human_VO_BR/NPCs/Shared/NPC_VO_Alerted_Att
 maxDistance=12500
 soundClassObject=&Fort_VO_Enemy_Critical
 volumeMultiplier=1
 concurrencySet [/NitroDrome/Assets/Audio/PA/VO/VO_PA_Announcer_Conc]
}
