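# Sound cue for Nitro Queen voice lines, named for the player exiting a POI vehicle in NitroDrome.
# A Random node picks among four equally weighted wave players, one per recorded variant, and is excluded from branch culling.
# Playback uses shared NPC alerted-VO attenuation, a critical enemy VO sound class, and the PA announcer concurrency set.
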SoundCue /NitroDrome/Assets/Audio/VO/Nitro_Queen/VO_Nitro_Queen_Player_Exits_POI_Veh_Cue
root=Random_2
nodes [
 {
  id=Random_2
  children [WavePlayer_8, WavePlayer_10, WavePlayer_11, WavePlayer_12]
  properties {bShouldExcludeFromBranchCulling=true, weights [1,1,1,1]}
 }
 {id=WavePlayer_10, properties {soundWaveAssetPtr=&/VO_Nitro_Queen_Player_Exits_Veh_02}}
 {id=WavePlayer_11, properties {soundWaveAssetPtr=&/VO_Nitro_Queen_Player_Exits_Veh_03}}
 {id=WavePlayer_12, properties {soundWaveAssetPtr=&/VO_Nitro_Queen_Player_Exits_Veh_04}}
 {id=WavePlayer_8, properties {soundWaveAssetPtr=&/VO_Nitro_Queen_Player_Exits_Veh_01}}
]
properties {
 attenuationSettings=/Game/Sounds/Fort_Human_VO_BR/NPCs/Shared/NPC_VO_Alerted_Att
 maxDistance=12500
 soundClassObject=&Fort_VO_Enemy_Critical
 volumeMultiplier=1
 concurrencySet [/NitroDrome/Assets/Audio/PA/VO/VO_PA_Announcer_Conc]
}
