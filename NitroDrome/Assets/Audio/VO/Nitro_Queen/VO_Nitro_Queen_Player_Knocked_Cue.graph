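# Sound cue for Nitro Queen voice lines, named for a player being knocked in NitroDrome.
# A single Random node picks among seven equally weighted wave players, excluded from branch culling, all using the Player_Elim recordings despite the Knocked name.
# Playback uses a shared NPC alerted-VO attenuation, the enemy critical VO sound class and the PA announcer concurrency set.
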SoundCue /NitroDrome/Assets/Audio/VO/Nitro_Queen/VO_Nitro_Queen_Player_Knocked_Cue
root=Random_1
nodes [
 {
  id=Random_1
  children [WavePlayer_7, WavePlayer_8, WavePlayer_9, WavePlayer_10, WavePlayer_11, WavePlayer_12, WavePlayer_13]
  properties {bShouldExcludeFromBranchCulling=true, weights [1,1,1,1,1,1,1]}
 }
 {id=WavePlayer_10, properties {soundWaveAssetPtr=&/VO_Nitro_Queen_Player_Elim_04}}
 {id=WavePlayer_11, properties {soundWaveAssetPtr=&/VO_Nitro_Queen_Player_Elim_05}}
 {id=WavePlayer_12, properties {soundWaveAssetPtr=&/VO_Nitro_Queen_Player_Elim_06}}
 {id=WavePlayer_13, properties {soundWaveAssetPtr=&/VO_Nitro_Queen_Player_Elim_07}}
 {id=WavePlayer_7, properties {soundWaveAssetPtr=&/VO_Nitro_Queen_Player_Elim_01}}
 {id=WavePlayer_8, properties {soundWaveAssetPtr=&/VO_Nitro_Queen_Player_Elim_02}}
 {id=WavePlayer_9, properties {soundWaveAssetPtr=&/VO_Nitro_Queen_Player_Elim_03}}
]
properties {
 attenuationSettings=/Game/Sounds/Fort_Human_VO_BR/NPCs/Shared/NPC_VO_Alerted_Att
 maxDistance=12500
 soundClassObject=&Fort_VO_Enemy_Critical
 volumeMultiplier=1
 concurrencySet [/NitroDrome/Assets/Audio/PA/VO/VO_PA_Announcer_Conc]
}
