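# Sound cue for Nitro Queen voice lines, apparently played when a player takes sawblade damage.
# A Random node picks among five Saw Blade takes with equal weights and is excluded from branch culling.
# It uses shared NPC alerted-VO attenuation and the critical enemy VO sound class.
# Its concurrency set is the PA announcer's, so the line shares voice limits with announcer VO.
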SoundCue /NitroDrome/Assets/Audio/VO/Nitro_Queen/VO_Nitro_Queen_Player_Sawblade_Dmg_Cue
root=Random_2
nodes [
 {
  id=Random_2
  children [WavePlayer_9, WavePlayer_13, WavePlayer_16, WavePlayer_17, WavePlayer_19]
  properties {bShouldExcludeFromBranchCulling=true, weights [1,1,1,1,1]}
 }
 {id=WavePlayer_13, properties {soundWaveAssetPtr=&/VO_Nitro_Queen_Saw_Blade_02}}
 {id=WavePlayer_16, properties {soundWaveAssetPtr=&/VO_Nitro_Queen_Saw_Blade_03}}
 {id=WavePlayer_17, properties {soundWaveAssetPtr=&/VO_Nitro_Queen_Saw_Blade_04}}
 {id=WavePlayer_19, properties {soundWaveAssetPtr=&/VO_Nitro_Queen_Saw_Blade_05}}
 {id=WavePlayer_9, properties {soundWaveAssetPtr=&/VO_Nitro_Queen_Saw_Blade_01}}
]
properties {
 attenuationSettings=/Game/Sounds/Fort_Human_VO_BR/NPCs/Shared/NPC_VO_Alerted_Att
 maxDistance=12500
 soundClassObject=&Fort_VO_Enemy_Critical
 volumeMultiplier=1
 concurrencySet [/NitroDrome/Assets/Audio/PA/VO/VO_PA_Announcer_Conc]
}
